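# Physics asset for the jacket of the male medium SmoothBeanie character body, defining capsule bodies and constraints for its dynamic jacket bones.
# Two-segment back, side and front chains hang from spine_03, with single-axis swing limits and tiny base capsules on free constraints before each front segment.
# Pelvis, spine and left thigh capsules form the body; there is no right thigh.
# Left front and side capsules are larger and offset, unlike the uniform right and back ones.
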PhysicsAsset /BRCosmetics/Characters/Player/Male/Medium/Bodies/M_MED_SmoothBeanie/Meshes/M_MED_SmoothBeanie_Jacket_Physics
bodies {
 base_front_jacket_l [{shapeType=Capsule, length=0.505, radius=0.505, shapeName=base_front_jacket_l_capsule, rotation=<rot 0 -90 0>}]
 base_front_jacket_l_02 [{shapeType=Capsule, length=0.505, radius=0.505, shapeName=base_front_jacket_l_02_capsule, rotation=<rot 0 -90 0>}]
 base_front_jacket_r [{shapeType=Capsule, length=0.505, radius=0.505, shapeName=base_front_jacket_r_capsule, rotation=<rot 0 -90 0>}]
 base_front_jacket_r_02 [{shapeType=Capsule, length=0.505, radius=0.505, shapeName=base_front_jacket_r_02_capsule, rotation=<rot 0 -90 0>}]
 dyn_back_jacket_m_01 [{shapeType=Capsule, length=2, radius=2, center=<2 0 0>, rotation=<rot 90 0 0>}]
 dyn_back_jacket_m_02 [{shapeType=Capsule, length=2, radius=2, center=<2 0 0>, rotation=<rot 90 0 0>}]
 dyn_front_jacket_l_01 [{shapeType=Capsule, length=5, radius=3.5, shapeName=dyn_jacket_l_fr_1_capsule, center=<4.5061 -1.3185 0.071>, rotation=<rot 0 0.0002 -50>}]
 dyn_front_jacket_l_02 [{shapeType=Capsule, length=5, radius=3, shapeName=dyn_jacket_l_fr_2_capsule, center=<4.4894 -0.9509 -0.2215>, rotation=<rot 0 0.0002 -50>}]
 dyn_front_jacket_r_01 [{shapeType=Capsule, length=2, radius=2, center=<2 0 0>, rotation=<rot 90 0 0>}]
 dyn_front_jacket_r_02 [{shapeType=Capsule, length=2, radius=2, center=<2 0 0>, rotation=<rot 90 0 0>}]
 dyn_side_jacket_l_01 [{shapeType=Capsule, length=5, radius=3.5, shapeName=dyn_jacket_l_1_capsule, center=<4.2436 -0.714 -2.9>}]
 dyn_side_jacket_l_02 [{shapeType=Capsule, length=5, radius=3.5, shapeName=dyn_jacket_l_2_capsule, center=<4.6569 -1.5229 -2.0394>}]
 dyn_side_jacket_r_01 [{shapeType=Capsule, length=2, radius=2, center=<2 0 0>, rotation=<rot 90 0 0>}]
 dyn_side_jacket_r_02 [{shapeType=Capsule, length=2, radius=2, center=<2 0 0>, rotation=<rot 90 0 0>}]
 pelvis [{shapeType=Capsule, length=7.75, radius=9.75, shapeName=pelvis_capsule, center=<-0.4848 -0.568 0.0758>}]
 spine_02 [{shapeType=Capsule, length=6, radius=9.6, shapeName=spine_02_capsule, center=<-2.9707 -1.4032 0>}]
 spine_03 [{shapeType=Capsule, length=6, radius=9.4, center=<0 -0.1341 0>}]
 thigh_l [{
   shapeType=Capsule
   length=25
   radius=7.9
   shapeName=thigh_l_capsule
   center=<-13.1328 2.8075 0.3699>
   rotation=<rot -0.306 -88.0507 -88.7674>
  }]
}
constraints [
 {
  bone1Name=spine_03
  bone2Name=spine_02
  swing1LimitDegrees=45
  swing1Motion=Free
  swing2LimitDegrees=45
  swing2Motion=Free
  twistLimitDegrees=45
  twistMotion=Free
 }
 {bone1Name=dyn_back_jacket_m_01, bone2Name=spine_03, swing1LimitDegrees=10, swing2LimitDegrees=0, twistLimitDegrees=0}
 {bone1Name=dyn_back_jacket_m_02, bone2Name=dyn_back_jacket_m_01, swing1LimitDegrees=20, swing2LimitDegrees=0, twistLimitDegrees=0}
 {bone1Name=dyn_side_jacket_r_01, bone2Name=spine_03, swing1LimitDegrees=10, swing2LimitDegrees=0, twistLimitDegrees=0}
 {bone1Name=dyn_side_jacket_r_02, bone2Name=dyn_side_jacket_r_01, swing1LimitDegrees=20, swing2LimitDegrees=0, twistLimitDegrees=0}
 {bone1Name=dyn_side_jacket_l_01, bone2Name=spine_03, swing1LimitDegrees=20, swing2LimitDegrees=0, twistLimitDegrees=0}
 {bone1Name=dyn_side_jacket_l_02, bone2Name=dyn_side_jacket_l_01, swing1LimitDegrees=20, swing2LimitDegrees=0, twistLimitDegrees=0}
 {
  bone1Name=thigh_l
  bone2Name=pelvis
  swing1LimitDegrees=45
  swing1Motion=Free
  swing2LimitDegrees=45
  swing2Motion=Free
  twistLimitDegrees=45
  twistMotion=Free
 }
 {
  bone1Name=base_front_jacket_l
  bone2Name=spine_03
  swing1LimitDegrees=45
  swing1Motion=Free
  swing2LimitDegrees=45
  swing2Motion=Free
  twistLimitDegrees=45
  twistMotion=Free
 }
 {
  bone1Name=base_front_jacket_l_02
  bone2Name=dyn_front_jacket_l_01
  swing1LimitDegrees=45
  swing1Motion=Free
  swing2LimitDegrees=45
  swing2Motion=Free
  twistLimitDegrees=45
  twistMotion=Free
 }
 {bone1Name=dyn_front_jacket_l_01, bone2Name=base_front_jacket_l, swing1LimitDegrees=0, swing2LimitDegrees=20, twistLimitDegrees=0}
 {bone1Name=dyn_front_jacket_l_02, bone2Name=base_front_jacket_l_02, swing1LimitDegrees=0, swing2LimitDegrees=20, twistLimitDegrees=0}
 {
  bone1Name=base_front_jacket_r
  bone2Name=spine_03
  swing1LimitDegrees=45
  swing1Motion=Free
  swing2LimitDegrees=45
  swing2Motion=Free
  twistLimitDegrees=45
  twistMotion=Free
 }
 {
  bone1Name=base_front_jacket_r_02
  bone2Name=dyn_front_jacket_r_01
  swing1LimitDegrees=45
  swing1Motion=Free
  swing2LimitDegrees=45
  swing2Motion=Free
  twistLimitDegrees=45
  twistMotion=Free
 }
 {bone1Name=dyn_front_jacket_r_01, bone2Name=base_front_jacket_r, swing1LimitDegrees=20, swing2LimitDegrees=0, twistLimitDegrees=0}
 {bone1Name=dyn_front_jacket_r_02, bone2Name=base_front_jacket_r_02, swing1LimitDegrees=20, swing2LimitDegrees=0, twistLimitDegrees=0}
 {
  bone1Name=spine_02
  bone2Name=pelvis
  swing1LimitDegrees=45
  swing1Motion=Free
  swing2LimitDegrees=45
  swing2Motion=Free
  twistLimitDegrees=45
  twistMotion=Free
 }
]
