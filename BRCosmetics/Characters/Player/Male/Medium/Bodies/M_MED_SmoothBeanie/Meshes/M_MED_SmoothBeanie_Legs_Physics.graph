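# Physics asset for the legs mesh of the SmoothBeanie medium male body, set up for dangling lock picks and a leg strap.
# Capsules on the pelvis and right thigh provide body volumes, with the thigh and lock-pick base joined to the pelvis by angularly free constraints.
# The main lock-pick body hangs from that base and branches into four picks with narrow twist and wider single-axis swing limits.
# The strap attaches to the pelvis as a hinge, swinging on one axis only.
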PhysicsAsset /BRCosmetics/Characters/Player/Male/Medium/Bodies/M_MED_SmoothBeanie/Meshes/M_MED_SmoothBeanie_Legs_Physics
bodies {
 base_LockPicks [{shapeType=Capsule, length=0.5, radius=0.5}]
 dyn_LockPicks_Main [{shapeType=Capsule, length=1, radius=1, center=<1 0 0.2605>, rotation=<rot 90 0 0>}]
 dyn_LockPicks_a [{shapeType=Capsule, length=5, radius=0.5, center=<4.1956 0 0>, rotation=<rot 90 0 0>}]
 dyn_LockPicks_b [{shapeType=Capsule, length=5, radius=0.5, center=<4.1956 0 0>, rotation=<rot 90 0 0>}]
 dyn_LockPicks_c [{shapeType=Capsule, length=5, radius=0.5, center=<4.1956 0 0>, rotation=<rot 90 0 0>}]
 dyn_LockPicks_d [{shapeType=Capsule, length=5, radius=0.5, center=<4.1956 0 0>, rotation=<rot 90 0 0>}]
 dyn_legStrap [{shapeType=Capsule, length=3, radius=3, center=<3 0 0>, rotation=<rot 90 0 0>}]
 pelvis [{shapeType=Capsule, length=11.2002, radius=10, center=<-2.0423 -1.1174 0.5631>}]
 thigh_r [{shapeType=Capsule, length=10.5, radius=10.5, center=<2.8462 1.0984 -2.5138>, rotation=<rot 90 0 0>}]
}
constraints [
 {
  bone1Name=thigh_r
  bone2Name=pelvis
  swing1LimitDegrees=45
  swing1Motion=Free
  swing2LimitDegrees=45
  swing2Motion=Free
  twistLimitDegrees=45
  twistMotion=Free
 }
 {
  bone1Name=base_LockPicks
  bone2Name=pelvis
  swing1LimitDegrees=45
  swing1Motion=Free
  swing2LimitDegrees=45
  swing2Motion=Free
  twistLimitDegrees=45
  twistMotion=Free
 }
 {bone1Name=dyn_LockPicks_Main, bone2Name=base_LockPicks, swing1LimitDegrees=2, swing2LimitDegrees=20, twistLimitDegrees=5}
 {bone1Name=dyn_LockPicks_a, bone2Name=dyn_LockPicks_Main, swing1LimitDegrees=10, swing2LimitDegrees=50, twistLimitDegrees=5}
 {bone1Name=dyn_LockPicks_b, bone2Name=dyn_LockPicks_Main, swing1LimitDegrees=10, swing2LimitDegrees=45, twistLimitDegrees=5}
 {bone1Name=dyn_LockPicks_c, bone2Name=dyn_LockPicks_Main, swing1LimitDegrees=10, swing2LimitDegrees=35, twistLimitDegrees=5}
 {bone1Name=dyn_LockPicks_d, bone2Name=dyn_LockPicks_Main, swing1LimitDegrees=10, swing2LimitDegrees=45, twistLimitDegrees=5}
 {bone1Name=dyn_legStrap, bone2Name=pelvis, swing1LimitDegrees=0, swing2LimitDegrees=30, twistLimitDegrees=0}
]
