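# Cosmetic AnimBP for the male SmoothBeanie body, adding secondary motion to jacket and legs.
# It copies the attached parent's pose into a deform Control Rig driven by sprinting, then blends jacket rigid-body physics with fixed bone offsets used when physics is off or LOD exceeds 1.
# Bone correctives follow, then leg rigid-body physics, leg-strap AnimDynamics and knee-helper offsets.
# The EventGraph updates movement, emote, gravity, front-end and knee-bend values.
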AnimBlueprint /BRCosmetics/Characters/Player/Male/Medium/Bodies/M_MED_SmoothBeanie/Meshes/M_MED_SmoothBeanie_AnimBP
target_skeleton=&/M_MED_SmoothBeanie_Skeleton
graphs [
 AnimGraph {nodes [
   "Transform (Modify) Bone - Bone: base_legStrap"!(ComponentPose, bAlphaBoolEnabled:bool=false, Translation="0,0,0", Rotation="-31.528,0,0", Scale="1,1,1") -> (Pose) {
    alphaInputType=Bool
    rotationMode=BMM_Additive
    rotationSpace=BCS_BoneSpace
    alphaBoolBlend {blendInTime=0.25, blendOption=Linear, blendOutTime=0.25}
    boneToModify {boneName=base_legStrap}
    rotation {pitch=-31.528}
   }
   n1 = RigidBody!(ComponentPose=Local To Component(LocalPose=Use cached pose 'CacheJAcket'()), Alpha, OverrideWorldGravity:Vector="0,0,-980") -> (Pose) {
    bClampLinearTranslationLimitToRefPose=true
    bOverrideWorldGravity=true
    overridePhysicsAsset=&/M_MED_SmoothBeanie_Legs_Physics
    simulationSpace=BaseBoneSpace
    alphaBoolBlend {blendOption=Linear}
    baseBoneRef {boneName=pelvis}
    componentAppliedLinearAccClamp=<5000 5000 5000>
    componentLinearAccScale=<0.8 0.8 0.35>
    componentLinearVelScale=<0.8 0.8 0.35>
    simSpaceSettings {maxAngularAcceleration=100, maxAngularVelocity=9, maxLinearAcceleration=100, maxLinearVelocity=300, velocityScaleZ=0, worldAlpha=1}
   }
   n2 = Anim Dynamics - Bone: dyn_legStrap!(SimulationSpace:AnimPhysSimSpaceType Enum, ComponentPose=@n1, Alpha=1, ExternalForce="0,0,0") -> (Pose) {
    angularDampingOverride=1
    angularSpringConstant=20000
    bAngularSpring=true
    bOverrideAngularDamping=true
    bUseGravityOverride=true
    gravityScale=0.35
    simulationSpace=RootRelative
    alphaBoolBlend {blendOption=Linear}
    boundBone {boneName=dyn_legStrap}
    componentAppliedLinearAccClamp=<10000 10000 10000>
    componentLinearAccScale=<1 1 1>
    componentLinearVelScale=<1 1 1>
    gravityOverride {z=-980}
    physicsBodyDefinitions [{
      boundBone {boneName=dyn_legStrap}
      boxExtents=<20 30 30>
      constraintSetup {angularLimitsMax=<15 80 15>, angularLimitsMin=<-15 1 -15>, angularTarget=<1 0 0>}
      localJointOffset=<4 0 0>
     }]
   }
   n3 = "Transform (Modify) Bone - Bone: dyn_Knee_Helper"(ComponentPose=@n2, Alpha=1, Translation="0,0,0", Rotation="0,0,-5", Scale="1,1,1") {
    rotationMode=BMM_Additive
    rotationSpace=BCS_BoneSpace
    alphaBoolBlend {blendOption=Linear}
    boneToModify {boneName=dyn_Knee_Helper}
    rotation {roll=-5}
   }
   n4 = "Transform (Modify) Bone - Bone: base_Knee_Helper"(ComponentPose=@n3, Alpha=1, Translation="0,-0.2,6", Rotation="0,0,-60", Scale="1,1,1") {
    rotationMode=BMM_Additive
    rotationSpace=BCS_BoneSpace
    translationMode=BMM_Additive
    translationSpace=BCS_BoneSpace
    alphaBoolBlend {blendOption=Linear}
    boneToModify {boneName=base_Knee_Helper}
    rotation {roll=-60}
    translation {y=-0.2, z=6}
   }
   Output Pose(Result=Component To Local(ComponentPose=@n4)) {layerGroup=None}
   n5 = Control Rig!(Source=Copy Pose From Mesh() {bUseAttachedParent=true}, Alpha, IsTacticalSprint:double=Get IsSprinting() -> (IsSprinting:double)) -> (Pose) {
    alphaBoolBlend {blendOption=Linear}
    controlRigAssetReference {blueprintRigClass=&/M_MED_SmoothBeanie_Deform_CtrlRig.M_MED_SmoothBeanie_Deform_CtrlRig_C}
   }
   Save cached pose 'CacheCR'(Pose=@n5) {cacheName=CacheCR}
   n6 = "Transform (Modify) Bone - Bone: base_side_jacket_l"!(
    ComponentPose=Local To Component(LocalPose=Use cached pose 'CacheCR'())
    bAlphaBoolEnabled:bool=True
    Translation="0,0,0"
    Rotation="10,0,0"
    Scale="1,1,1"
   ) -> (Pose) {
    alphaInputType=Bool
    rotationMode=BMM_Additive
    rotationSpace=BCS_BoneSpace
    translationMode=BMM_Additive
    translationSpace=BCS_BoneSpace
    alphaBoolBlend {blendInTime=0.25, blendOption=Linear, blendOutTime=0.25}
    boneToModify {boneName=base_side_jacket_l}
    rotation {pitch=10}
   }
   n7 = "Transform (Modify) Bone - Bone: base_front_jacket_r"!(ComponentPose=@n6, bAlphaBoolEnabled:bool=True, Translation="0,0,0", Rotation="-10,0,0", Scale="1,1,1") -> (Pose) {
    alphaInputType=Bool
    rotationMode=BMM_Additive
    rotationSpace=BCS_BoneSpace
    translationMode=BMM_Additive
    translationSpace=BCS_BoneSpace
    alphaBoolBlend {blendInTime=0.25, blendOption=Linear, blendOutTime=0.25}
    boneToModify {boneName=base_front_jacket_r}
    rotation {pitch=-10}
   }
   n8 = "Transform (Modify) Bone - Bone: base_back_jacket_m"!(ComponentPose=@n7, bAlphaBoolEnabled:bool=True, Translation="-2.569,-0.019,2.766", Rotation="15,0,0", Scale="1,1,1") -> (Pose) {
    alphaInputType=Bool
    rotationMode=BMM_Additive
    rotationSpace=BCS_BoneSpace
    translationMode=BMM_Additive
    translationSpace=BCS_BoneSpace
    alphaBoolBlend {blendInTime=0.25, blendOption=Linear, blendOutTime=0.25}
    boneToModify {boneName=base_back_jacket_m}
    rotation {pitch=15}
    translation=<-2.569 -0.019 2.766>
   }
   n9 = Component To Local(ComponentPose=@n8)
   n10 = RigidBody(ComponentPose=Local To Component(LocalPose=Use cached pose 'CacheCR'()), OverrideWorldGravity="0,0,-980") {
    bClampLinearTranslationLimitToRefPose=true
    bOverrideWorldGravity=true
    overridePhysicsAsset=&/M_MED_SmoothBeanie_Jacket_Physics
    simulationSpace=BaseBoneSpace
    alphaBoolBlend {blendOption=Linear}
    baseBoneRef {boneName=pelvis}
    componentLinearAccScale {x=0.35, z=0.8}
    componentLinearVelScale {x=0.35, z=0.8}
    simSpaceSettings {
     maxAngularAcceleration=500
     maxAngularVelocity=3
     maxLinearAcceleration=400
     maxLinearVelocity=250
     velocityScaleZ=0
     worldAlpha=1
     externalLinearDragV=<0.1 0.1>
    }
   }
   n11 = Component To Local(ComponentPose=@n10)
   n12 = Get isRigidBodyDisabled() -> (isRigidBodyDisabled:bool)
   n13 = "Transform (Modify) Bone - Bone: base_front_jacket_r"(
    ComponentPose=Local To Component(LocalPose=Blend Poses by bool(BlendPose_0=@n9, BlendPose_1=@n11, bActiveValue=@n12) {blendType=Linear})
    bAlphaBoolEnabled=false
    Translation="0,0,0"
    Rotation="0.005,61.435,0.014"
    Scale="1,1,1"
   ) {
    alphaInputType=Bool
    rotationMode=BMM_Additive
    rotationSpace=BCS_BoneSpace
    scaleSpace=BCS_BoneSpace
    alphaBoolBlend {blendInTime=0.25, blendOption=Linear, blendOutTime=0.25}
    boneToModify {boneName=base_front_jacket_r}
    rotation=<rot 0.005 61.435 0.014>
   }
   n14 = "Transform (Modify) Bone - Bone: deform_upperarm_up_base_l"!(ComponentPose=@n13, bAlphaBoolEnabled:bool=True, Translation="0,0,-10", Rotation="0,0,0", Scale="1,1,1") -> (Pose) {
    alphaInputType=Bool
    rotationMode=BMM_Additive
    rotationSpace=BCS_BoneSpace
    translationMode=BMM_Additive
    translationSpace=BCS_BoneSpace
    alphaBoolBlend {blendInTime=0.25, blendOption=Linear, blendOutTime=0.25}
    boneToModify {boneName=deform_upperarm_up_base_l}
    translation {z=-10}
   }
   n15 = "Transform (Modify) Bone - Bone: deform_upperarm_up_base_r"!(ComponentPose=@n14, bAlphaBoolEnabled:bool=True, Translation="0,0,15", Rotation="0,0,0", Scale="1,1,1") -> (Pose) {
    alphaInputType=Bool
    rotationMode=BMM_Additive
    rotationSpace=BCS_BoneSpace
    translationMode=BMM_Additive
    translationSpace=BCS_BoneSpace
    alphaBoolBlend {blendInTime=0.25, blendOption=Linear, blendOutTime=0.25}
    boneToModify {boneName=deform_upperarm_up_base_r}
    translation {z=15}
   }
   n16 = "Transform (Modify) Bone - Bone: upperarm_twist_02_l"!(ComponentPose=@n15, bAlphaBoolEnabled:bool=True, Translation="0,0,0", Rotation="0,0,0", Scale="0.825,0.825,0.825") -> (Pose) {
    alphaInputType=Bool
    rotationSpace=BCS_BoneSpace
    scaleMode=BMM_Additive
    scaleSpace=BCS_BoneSpace
    translationSpace=BCS_BoneSpace
    alphaBoolBlend {blendInTime=0.25, blendOption=Linear, blendOutTime=0.25}
    boneToModify {boneName=upperarm_twist_02_l}
    scale=<0.825 0.825 0.825>
   }
   n17 = "Transform (Modify) Bone - Bone: base_front_jacket_r"!(ComponentPose=@n16, Alpha:float=1, Translation="0,5,0", Rotation="0,20,0", Scale="1,1,1") -> (Pose) {
    rotationMode=BMM_Additive
    rotationSpace=BCS_BoneSpace
    translationMode=BMM_Additive
    translationSpace=BCS_BoneSpace
    alphaBoolBlend {blendOption=Linear}
    boneToModify {boneName=base_front_jacket_r}
    rotation {yaw=20}
    translation {y=5}
   }
   n18 = "Transform (Modify) Bone - Bone: base_side_jacket_l"!(ComponentPose=@n17, Alpha:float=1, Translation="0,5,5", Rotation="0,0,0", Scale="1,1,1") -> (Pose) {
    rotationMode=BMM_Additive
    rotationSpace=BCS_BoneSpace
    translationMode=BMM_Additive
    translationSpace=BCS_BoneSpace
    alphaBoolBlend {blendOption=Linear}
    boneToModify {boneName=base_side_jacket_l}
    translation {y=5, z=5}
   }
   n19 = "Transform (Modify) Bone - Bone: base_front_jacket_r"(ComponentPose=@n18, Alpha=1, Translation="0,2,0", Rotation="0,0,0", Scale="1,1,1") {
    translationMode=BMM_Additive
    translationSpace=BCS_BoneSpace
    alphaBoolBlend {blendOption=Linear}
    boneToModify {boneName=base_front_jacket_r}
    translation {y=2}
   }
   n20 = "Transform (Modify) Bone - Bone: base_front_jacket_l"(ComponentPose=@n19, Alpha=1, Translation="0,2,0", Rotation="0,0,0", Scale="1,1,1") {
    translationMode=BMM_Additive
    translationSpace=BCS_BoneSpace
    alphaBoolBlend {blendOption=Linear}
    boneToModify {boneName=base_front_jacket_l}
    translation {y=2}
   }
   Save cached pose 'CacheJAcket'(Pose=Component To Local(ComponentPose=@n20)) {cacheName=CacheJAcket}
  ]}
 EventGraph {nodes [
   Event BlueprintUpdateAnimation()
   > n1 = Sequence!(execute) -> (then_0, then_1, then_2:Exec, then_3:Exec, then_4:Exec)
   n2 = GetOwnerSkeletalMeshComponent()
   n3 = GetAnimInstance(self=@n2)
   > Is Valid(InputObject=@n3)
   > n4 = Sequence!(execute) -> (then_0, then_1)
   > Set isCustomEmotePlaying(isCustomEmotePlaying:double=SelectFloat(A=1, B=0, bPickA=Montage_IsPlaying(self=@n3, Montage=/BRCosmetics/Animation/Game/MainPlayer/Emotes/Eastern_Bloc/Emote_EasternBloc_CMM_M))) -> (Output_Get:double)
   Set isCustomEmotePlaying_b(
    execute=@n4.then_1
    isCustomEmotePlaying_b:double=SelectFloat(A=1, B=0, bPickA=Montage_IsPlaying(self=@n3, Montage=/Game/Animation/Game/MainPlayer/Emotes/Emote_Celebration_M))
   ) -> (Output_Get:double)
   n5 = Gravity Override State Function DEPRECATED(
    execute=@n1.then_1
    CharacterPartAnimInstance=Self-Reference()
    GameTimeInSeconds=Get Game Time in Seconds() -> (Game Time in Seconds:double)
    GravityOverrideParams=Get Gravity Override Params()
   )
   > Set Gravity Override(Gravity Override=@n5)
   > Set ISCrouchingActive(ISCrouchingActive:bool=Get bIsCrouching()) -> (Output_Get:bool)
   > Set IsSlidingActive(IsSlidingActive:bool=OR Boolean!(A=Get bIsKneeSliding(), B=Get bIsSliding(), C:bool=Get bIsBaseballSliding()) -> (ReturnValue)) -> (Output_Get:bool)
   > Set IsParachuteActive(IsParachuteActive:bool=Get bIsParachuteOpened()) -> (Output_Get:bool)
   > Set IsSprinting(IsSprinting:double="To Float (Boolean)"(InBool=Get bIsSprinting())) -> (Output_Get:double)
   n6 = Branch(execute=@n1.then_3, Condition=Get bIsInFrontEnd())
   > Set SimlationSpace(SimlationSpace:AnimPhysSimSpaceType Enum=World) -> (Output_Get:AnimPhysSimSpaceType Enum)
   > Set AlphaFE(AlphaFE:double=0.3) -> (Output_Get:double)
   Set SimlationSpace(execute=@n6.else)
   > Set AlphaFE(AlphaFE=1)
   n7 = Get bRigidBodyEnabled()
   n8 = Get CurrentLOD()
   Set isRigidBodyDisabled(
    execute=@n1.then_4
    isRigidBodyDisabled:bool=OR Boolean!(A=NOT Boolean(A=@n7), B="integer > integer"(A=@n8, B=1) {is=title K2Node_CallFunction}) -> (ReturnValue)
   ) -> (Output_Get:bool)
   Set Knee_Brace(
    execute=@n1.then_2
    Knee_Brace:double=GetAlphaFromReferenceTransformDelta(SkeletalMeshComponent=@n2, BoneName=calf_r, BaseName=thigh_r, InputRange=(X=0,Y=120), TwistAxis="0,0,1")
   ) -> (Output_Get:double)
  ]}
]
variables [
 AlphaFE {type=number}
 Game Time in Seconds {type=number}
 Gravity Override {type=Vector}
 Gravity Override Params {type=GravityOverrideParamsStruct}
 ISCrouchingActive {type=boolean}
 isCustomEmotePlaying {type=number}
 isCustomEmotePlaying_b {type=number}
 IsParachuteActive {type=boolean, default=true}
 isRigidBodyDisabled {type=boolean}
 IsSlidingActive {type=boolean, default=true}
 IsSprinting {type=number}
 Knee_Brace {type=number}
 Rigid Body Enabled {type=boolean}
 SimlationSpace {type=AnimPhysSimSpaceType}
]
